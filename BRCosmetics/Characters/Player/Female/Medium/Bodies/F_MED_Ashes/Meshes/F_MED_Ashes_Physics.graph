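# Physics asset for the F_MED_Ashes female character body, covering dangling straps on the right side plus the pelvis and right thigh.
# Each listed bone carries one capsule body, with only the pelvis and thigh_r representing the body itself.
# Two strap chains, A with four links and B with three, hang from strapMain through constraints with tight twist and moderate swing limits.
# strapMain and thigh_r join the pelvis with free angular motion.
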PhysicsAsset /BRCosmetics/Characters/Player/Female/Medium/Bodies/F_MED_Ashes/Meshes/F_MED_Ashes_Physics
bodies {
 pelvis [{shapeType=Capsule, length=12.75, radius=10, shapeName=pelvis_capsule, center=<-7.4868 0.0001 0>}]
 strapA_Rt1 [{shapeType=Capsule, length=6.5, radius=3.25, shapeName=strapA_Rt1_capsule, center=<5 0 -3.25>, rotation=<rot -90 0 90.0004>}]
 strapA_Rt2 [{shapeType=Capsule, length=18, radius=3.5, shapeName=strapA_Rt2_capsule, center=<5 0 -9.4317>, rotation=<rot -90 10.0248 -10.0244>}]
 strapA_Rt3 [{shapeType=Capsule, length=18, radius=3.5, shapeName=strapA_Rt3_capsule, center=<4.9999 0 -9.0817>, rotation=<rot -90 10.0248 -10.0244>}]
 strapA_Rt4 [{shapeType=Capsule, length=15.75, radius=3.5, shapeName=strapA_Rt4_capsule, center=<5 0 -8.8817>, rotation=<rot -90 0.0066 -0.0058>}]
 strapB_Rt1 [{shapeType=Capsule, length=6.5, radius=3.25, shapeName=strapB_Rt1_capsule, center=<5 0 -3.25>, rotation=<rot -90 0 90.0005>}]
 strapB_Rt2 [{shapeType=Capsule, length=20, radius=3.5, shapeName=strapB_Rt2_capsule, center=<5 0 -2.9>, rotation=<rot -90 0 90>}]
 strapB_Rt3 [{shapeType=Capsule, length=20, radius=3.5, shapeName=strapB_Rt3_capsule, center=<5 0 -2.95>, rotation=<rot -90 -90 -179.9995>}]
 strapMain [{
   shapeType=Capsule
   length=6.4031
   radius=5.8731
   shapeName=strapMain_capsule
   center=<2.1247 1.2969 -1.077>
   rotation=<rot 2.2963 -0.7594 -90>
  }]
 thigh_r [{
   shapeType=Capsule
   length=30.6811
   radius=7.6251
   shapeName=thigh_r_capsule
   center=<24.4216 1.3553 -0.3849>
   rotation=<rot 0 -83.9477 -86.3363>
  }]
}
constraints [
 {bone1Name=strapA_Rt1, bone2Name=strapMain, swing1LimitDegrees=45, swing2LimitDegrees=20, twistLimitDegrees=5}
 {bone1Name=strapA_Rt2, bone2Name=strapA_Rt1, swing1LimitDegrees=45, swing2LimitDegrees=20, twistLimitDegrees=5}
 {bone1Name=strapA_Rt3, bone2Name=strapA_Rt2, swing1LimitDegrees=45, swing2LimitDegrees=20, twistLimitDegrees=5}
 {bone1Name=strapA_Rt4, bone2Name=strapA_Rt3, swing1LimitDegrees=45, swing2LimitDegrees=20, twistLimitDegrees=5}
 {bone1Name=strapB_Rt1, bone2Name=strapMain, swing1LimitDegrees=45, swing2LimitDegrees=20, twistLimitDegrees=5}
 {bone1Name=strapB_Rt2, bone2Name=strapB_Rt1, swing1LimitDegrees=45, swing2LimitDegrees=20, twistLimitDegrees=5}
 {bone1Name=strapB_Rt3, bone2Name=strapB_Rt2, swing1LimitDegrees=45, swing2LimitDegrees=20, twistLimitDegrees=5}
 {
  bone1Name=thigh_r
  bone2Name=pelvis
  swing1LimitDegrees=45
  swing1Motion=Free
  swing2LimitDegrees=45
  swing2Motion=Free
  twistLimitDegrees=45
  twistMotion=Free
 }
 {
  bone1Name=strapMain
  bone2Name=pelvis
  swing1LimitDegrees=45
  swing1Motion=Free
  swing2LimitDegrees=45
  swing2Motion=Free
  twistLimitDegrees=45
  twistMotion=Free
 }
]
