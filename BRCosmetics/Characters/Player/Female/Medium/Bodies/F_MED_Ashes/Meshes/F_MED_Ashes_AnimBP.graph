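# Post-process style cosmetic AnimBP for the F_MED_Ashes body that copies the attached parent mesh's pose and adds deformation and secondary motion.
# Two Textile_Sparkle Control Rigs feed bone constraints and additive offsets on the thigh strap, clip and chest, then pelvis-space RigidBody physics and front-end-only strap offsets.
# The update event derives offset alphas from right-thigh rotation relative to pelvis, crouch-walking and front-end state; the LegRtSideRot strap node is unconnected.
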AnimBlueprint /BRCosmetics/Characters/Player/Female/Medium/Bodies/F_MED_Ashes/Meshes/F_MED_Ashes_AnimBP
target_skeleton=&/F_MED_Ashes_Skeleton
graphs [
 AnimGraph {nodes [
   "Transform (Modify) Bone - Bone: strapMain"(Alpha=Get LegRtSideRot() -> (LegRtSideRot:double), Translation="0,0,0", Rotation="-35,0,0", Scale="1,1,1") {
    rotationMode=BMM_Additive
    rotationSpace=BCS_BoneSpace
    scaleMode=BMM_Additive
    scaleSpace=BCS_BoneSpace
    translationMode=BMM_Additive
    translationSpace=BCS_BoneSpace
    alphaBoolBlend {blendOption=Linear}
    boneToModify {boneName=strapMain}
    rotation {pitch=-35}
   }
   n1 = Control Rig(Source=Copy Pose From Mesh() {bUseAttachedParent=true}) {
    alphaBoolBlend {blendOption=Linear}
    controlRigAssetReference {blueprintRigClass=/BRCosmetics/Characters/Player/Female/Medium/Bodies/F_MED_Textile_Sparkle/Meshes/F_MED_Textile_Sparkle_Deform_CtrlRig_twists.F_MED_Textile_Sparkle_Deform_CtrlRig_twists_C}
   }
   n2 = Control Rig!(
    Source=@n1
    Alpha
    KneeRot_fr_l_114deg:Vector
    KneeTrans_fr_l_114deg:Vector="0.258,-3.591,0.009"
    KneeScale_fr_l_114deg:Vector="1,1,1"
    KneeScale_fr_l_130deg:Vector="1,1,1"
    KneeRot_fr_l_130deg:Vector="0,0,-51.2"
    KneeTrans_fr_l_130deg:Vector="0.258,-3.591,0.009"
   ) -> (Pose) {
    alphaBoolBlend {blendOption=Linear}
    controlRigAssetReference {blueprintRigClass=/BRCosmetics/Characters/Player/Female/Medium/Bodies/F_MED_Textile_Sparkle/Meshes/F_MED_Textile_Sparkle_Deform_CtrlRig_sculpters.F_MED_Textile_Sparkle_Deform_CtrlRig_sculpters_C}
   }
   n3 = "Constraint - strapMain (RR)"!(ComponentPose=Local To Component(LocalPose=@n2), Alpha=1, ConstraintWeights_0=0.5, ConstraintWeights_1:float=0.5) -> (Pose) {
    alphaBoolBlend {blendOption=Linear}
    boneToModify {boneName=strapMain}
    constraintSetup [{transformType=Rotation, targetBone {boneName=pelvis}}, {transformType=Rotation, targetBone {boneName=thigh_r}}]
    constraintWeights [1,1]
   }
   n4 = "Constraint - strapMain (TT)"!(ComponentPose=@n3, Alpha=0, ConstraintWeights_0=0, ConstraintWeights_1:float=1) -> (Pose) {
    alphaBoolBlend {blendOption=Linear}
    boneToModify {boneName=strapMain}
    constraintSetup [{transformType=Translation, targetBone {boneName=pelvis}}, {transformType=Translation, targetBone {boneName=thigh_r}}]
    constraintWeights [1,1]
   }
   n5 = "Transform (Modify) Bone - Bone: strapMain"(ComponentPose=@n4, Alpha=Get LegRtRot() -> (LegRtRot:double), Translation="0,0,0", Rotation="-60,0,0", Scale="1,1,1") {
    rotationMode=BMM_Additive
    rotationSpace=BCS_BoneSpace
    scaleMode=BMM_Additive
    scaleSpace=BCS_BoneSpace
    translationMode=BMM_Additive
    translationSpace=BCS_BoneSpace
    alphaBoolBlend {blendOption=Linear}
    boneToModify {boneName=strapMain}
    rotation {pitch=-60}
   }
   n6 = "Transform (Modify) Bone - Bone: strapMain"(ComponentPose=@n5, Alpha=Get LegRtRot(), Translation="2,0,-1", Rotation="0,0,0", Scale="1,1,1") {
    rotationMode=BMM_Additive
    rotationSpace=BCS_BoneSpace
    scaleMode=BMM_Additive
    scaleSpace=BCS_BoneSpace
    translationMode=BMM_Additive
    translationSpace=BCS_BoneSpace
    alphaBoolBlend {blendOption=Linear}
    boneToModify {boneName=strapMain}
    translation {x=2, z=-1}
   }
   n7 = "Transform (Modify) Bone - Bone: strapClip"(ComponentPose=@n6, Alpha=Get isCrouchWalking() -> (isCrouchWalking:double), Translation="0,-3,1", Rotation="30,0,30", Scale="1,1,1") {
    rotationMode=BMM_Additive
    rotationSpace=BCS_BoneSpace
    scaleMode=BMM_Additive
    scaleSpace=BCS_BoneSpace
    translationMode=BMM_Additive
    translationSpace=BCS_BoneSpace
    alphaBoolBlend {blendOption=Linear}
    alphaScaleBiasClamp {bInterpResult=true}
    boneToModify {boneName=strapClip}
    rotation {pitch=30, roll=30}
    translation {y=-3, z=1}
   }
   n8 = "Constraint - strapClip (RR)"!(ComponentPose=@n7, Alpha=0.75, ConstraintWeights_0=0, ConstraintWeights_1:float=1) -> (Pose) {
    alphaBoolBlend {blendOption=Linear}
    boneToModify {boneName=strapClip}
    constraintSetup [{transformType=Rotation, targetBone {boneName=pelvis}}, {transformType=Rotation, targetBone {boneName=strapMain}}]
    constraintWeights [1,1]
   }
   n9 = "Transform (Modify) Bone - Bone: strapMain"(ComponentPose=@n8, Alpha=Get isCrouchWalking(), Translation="0,0,0", Rotation="5,0,0", Scale="1,1,1") {
    rotationMode=BMM_Additive
    rotationSpace=BCS_BoneSpace
    scaleMode=BMM_Additive
    scaleSpace=BCS_BoneSpace
    translationMode=BMM_Additive
    translationSpace=BCS_BoneSpace
    alphaBoolBlend {blendOption=Linear}
    alphaScaleBiasClamp {bInterpResult=true, interpSpeedDecreasing=100, interpSpeedIncreasing=100}
    boneToModify {boneName=strapMain}
    rotation {pitch=5}
   }
   n10 = "Constraint - chest_l (TT)"!(ComponentPose=@n9, Alpha=0.6, ConstraintWeights_0=0.5, ConstraintWeights_1:float=0.5) -> (Pose) {
    alphaBoolBlend {blendOption=Linear}
    boneToModify {boneName=chest_l}
    constraintSetup [{transformType=Translation, targetBone {boneName=upperarm_l}}, {transformType=Translation, targetBone {boneName=spine_04}}]
    constraintWeights [1,1]
   }
   n11 = "Constraint - chest_r (TT)"!(ComponentPose=@n10, Alpha=0.6, ConstraintWeights_0=0.5, ConstraintWeights_1:float=0.5) -> (Pose) {
    alphaBoolBlend {blendOption=Linear}
    boneToModify {boneName=chest_r}
    constraintSetup [{transformType=Translation, targetBone {boneName=upperarm_r}}, {transformType=Translation, targetBone {boneName=spine_04}}]
    constraintWeights [1,1]
   }
   n12 = "Constraint - chest_l (RR)"!(ComponentPose=@n11, Alpha=1, ConstraintWeights_0=0.5, ConstraintWeights_1:float=0.5) -> (Pose) {
    alphaBoolBlend {blendOption=Linear}
    boneToModify {boneName=chest_l}
    constraintSetup [{transformType=Rotation, targetBone {boneName=spine_04}}, {transformType=Rotation, targetBone {boneName=spine_05}}]
    constraintWeights [1,1]
   }
   n13 = "Constraint - chest_r (RR)"!(ComponentPose=@n12, Alpha=1, ConstraintWeights_0=0.5, ConstraintWeights_1:float=0.5) -> (Pose) {
    alphaBoolBlend {blendOption=Linear}
    boneToModify {boneName=chest_r}
    constraintSetup [{transformType=Rotation, targetBone {boneName=spine_04}}, {transformType=Rotation, targetBone {boneName=spine_05}}]
    constraintWeights [1,1]
   }
   n14 = RigidBody(ComponentPose=@n13) {
    bClampLinearTranslationLimitToRefPose=true
    bOverrideWorldGravity=true
    overridePhysicsAsset=&/F_MED_Ashes_Physics
    simulationSpace=BaseBoneSpace
    alphaBoolBlend {blendOption=Linear}
    baseBoneRef {boneName=pelvis}
    componentAppliedLinearAccClamp=<3000 3000 3000>
    componentLinearAccScale=<0.4 0.3 0.3>
    componentLinearVelScale=<0.4 0.3 0.3>
    overrideWorldGravity {z=-980}
   }
   n15 = Get IsFrontEnd() -> (IsFrontEnd:double)
   n16 = "Transform (Modify) Bone - Bone: strapA_Rt2"(ComponentPose=@n14, Alpha=@n15, Translation="15,0,0", Rotation="5,0,0", Scale="1,1,1") {
    rotationMode=BMM_Additive
    rotationSpace=BCS_BoneSpace
    scaleMode=BMM_Additive
    scaleSpace=BCS_BoneSpace
    translationMode=BMM_Additive
    translationSpace=BCS_BoneSpace
    alphaBoolBlend {blendOption=Linear}
    alphaScaleBiasClamp {bInterpResult=true, interpSpeedDecreasing=100, interpSpeedIncreasing=100}
    boneToModify {boneName=strapA_Rt2}
    rotation {pitch=5}
    translation {x=15}
   }
   n17 = "Transform (Modify) Bone - Bone: strapB_Rt2"(ComponentPose=@n16, Alpha=@n15, Translation="15,0,0", Rotation="5,0,0", Scale="1,1,1") {
    rotationMode=BMM_Additive
    rotationSpace=BCS_BoneSpace
    scaleMode=BMM_Additive
    scaleSpace=BCS_BoneSpace
    translationMode=BMM_Additive
    translationSpace=BCS_BoneSpace
    alphaBoolBlend {blendOption=Linear}
    alphaScaleBiasClamp {bInterpResult=true, interpSpeedDecreasing=100, interpSpeedIncreasing=100}
    boneToModify {boneName=strapB_Rt2}
    rotation {pitch=5}
    translation {x=15}
   }
   Output Pose(Result=Component To Local(ComponentPose=@n17)) {layerGroup=None}
  ]}
 EventGraph {nodes [
   Event BlueprintUpdateAnimation()
   > Set LegRtRot(LegRtRot:double=GetAlphaFromReferenceTransformDelta(SkeletalMeshComponent=GetOwnerSkeletalMeshComponent(), BoneName=thigh_r, BaseName=pelvis, InputRange=(X=-20,Y=-100), TwistAxis="1,0,0")) -> (Output_Get:double)
   > Set LegRtSideRot(LegRtSideRot:double=GetAlphaFromReferenceTransformDelta(BoneName=thigh_r, BaseName=pelvis, DeltaComponentType=EulerY, InputRange=(X=-20,Y=-100), TwistAxis="1,0,0")) -> (Output_Get:double)
   n1 = Get bIsCrouching()
   n2 = Get bIsMoving2D()
   > Set isCrouchWalking(isCrouchWalking:double="To Float (Boolean)"(InBool=AND Boolean(A=@n1, B=@n2))) -> (Output_Get:double)
   > Set IsFrontEnd(IsFrontEnd:double="To Float (Boolean)"(InBool=Get bIsInFrontEnd())) -> (Output_Get:double)
  ]}
]
variables [
 Is Targeting {type=number}
 isCrouchWalking {type=number}
 IsFrontEnd {type=number}
 LegRtRot {type=number}
 LegRtSideRot {type=number}
]
